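# Post-process style AnimBlueprint for the Ponytail 01 backpack cosmetic, adding secondary motion to its sword, two knives and flap.
# The input pose gets an additive bone-space nudge on backpack_root, then passes through four chained single-body AnimDynamics nodes.
# Stiff angular springs keep sword and knives within narrow limits, while the flap swings one way on one axis.
# Simulation spaces differ per node, and the event graph holds no logic.
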AnimBlueprint /BRCosmetics/Accessories/FORT_Backpacks/F_MED_Ponytail_01_Backpack/Meshes/F_MED_Ponytail_01_Backpack_Skeleton_AnimBlueprint
target_skeleton=&/F_MED_Ponytail_01_Backpack_Skeleton
graphs [
 AnimGraph {nodes [
   AnimGraphNode_SubInput_0 = Input Pose()
   n1 = "Transform (Modify) Bone - Bone: backpack_root"(ComponentPose=Local To Component(LocalPose=@AnimGraphNode_SubInput_0), Alpha=1, Translation="0,-2.5,0", Rotation="0,0,0", Scale="1,1,1") {
    translationMode=BMM_Additive
    translationSpace=BCS_BoneSpace
    alphaBoolBlend {blendOption=Linear}
    boneToModify {boneName=backpack_root}
    translation {y=-2.5}
   }
   n2 = Anim Dynamics - Bone: sword_dyn(ComponentPose=@n1, Alpha=1, ExternalForce="0,0,0") {
    alpha=0.9
    angularDampingOverride=0.98
    angularSpringConstant=450
    bAngularSpring=true
    bGravityOverrideInSimSpace=true
    bOverrideAngularDamping=true
    lODThreshold=1
    numSolverIterationsPostUpdate=4
    numSolverIterationsPreUpdate=16
    simulationSpace=World
    alphaBoolBlend {blendOption=Linear}
    boundBone {boneName=sword_dyn}
    componentLinearVelScale=<3 3 3>
    physicsBodyDefinitions [{
      sphereCollisionRadius=0
      boundBone {boneName=sword_dyn}
      boxExtents=<15 5 5>
      constraintSetup {angularTargetAxis=AxisZ, angularLimitsMax=<0 2.5 2.5>, angularLimitsMin=<0 -2.5 -2.5>, angularTarget=<0 0 1>}
      localJointOffset=<0 0 0>
     }]
    relativeSpaceBone {boneName=backpack_root}
   }
   n3 = Anim Dynamics - Bone: knife_dyn_l(ComponentPose=@n2, Alpha=1, ExternalForce="0,0,0") {
    alpha=0.9
    angularDampingOverride=0.98
    angularSpringConstant=190
    bAngularSpring=true
    bGravityOverrideInSimSpace=true
    bOverrideAngularDamping=true
    lODThreshold=1
    numSolverIterationsPostUpdate=4
    numSolverIterationsPreUpdate=16
    simulationSpace=Actor
    alphaBoolBlend {blendOption=Linear}
    boundBone {boneName=knife_dyn_l}
    physicsBodyDefinitions [{
      sphereCollisionRadius=0
      boundBone {boneName=knife_dyn_l}
      boxExtents=<15 5 5>
      constraintSetup {angularTargetAxis=AxisZ, angularLimitsMax=<0 3 3>, angularLimitsMin=<0 -3 -3>, angularTarget=<0 0 1>}
      localJointOffset=<0 -0.5 0>
     }]
   }
   n4 = Anim Dynamics - Bone: knife_dyn_r(ComponentPose=@n3, Alpha=1, ExternalForce="0,0,0") {
    alpha=0.9
    angularDampingOverride=0.98
    angularSpringConstant=380
    bAngularSpring=true
    bGravityOverrideInSimSpace=true
    bOverrideAngularDamping=true
    lODThreshold=1
    numSolverIterationsPostUpdate=4
    numSolverIterationsPreUpdate=16
    simulationSpace=World
    alphaBoolBlend {blendOption=Linear}
    boundBone {boneName=knife_dyn_r}
    physicsBodyDefinitions [{
      sphereCollisionRadius=0
      boundBone {boneName=knife_dyn_r}
      boxExtents=<15 5 5>
      constraintSetup {angularTargetAxis=AxisZ, angularLimitsMax=<0 3 3>, angularLimitsMin=<0 -3 -3>, angularTarget=<0 0 1>}
      localJointOffset=<0 0 0>
     }]
   }
   n5 = Anim Dynamics - Bone: flap_dyn(ComponentPose=@n4, Alpha=1, ExternalForce="0,0,0") {
    alpha=0.9
    angularDampingOverride=0.98
    angularSpringConstant=1000
    bAngularSpring=true
    bGravityOverrideInSimSpace=true
    bLinearSpring=true
    bOverrideAngularDamping=true
    lODThreshold=1
    linearSpringConstant=1000
    numSolverIterationsPostUpdate=4
    numSolverIterationsPreUpdate=16
    alphaBoolBlend {blendOption=Linear}
    boundBone {boneName=flap_dyn}
    componentAppliedLinearAccClamp=<1000 1000 1000>
    componentLinearAccScale=<1 1 1>
    componentLinearVelScale=<1 1 1>
    physicsBodyDefinitions [{
      sphereCollisionRadius=0
      boundBone {boneName=flap_dyn}
      boxExtents=<5 5 5>
      constraintSetup {angularTargetAxis=AxisZ, angularLimitsMax=<0 0 30>, angularLimitsMin=<0 0 0>, angularTarget=<0 0 0>}
      localJointOffset=<0 0 0>
     }]
    relativeSpaceBone {boneName=backpack_root}
   }
   Output Pose(Result=Component To Local(ComponentPose=@n5)) {layerGroup=None}
  ]}
 EventGraph {nodes [Event BlueprintUpdateAnimation()]}
]
variables []
